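# Material function for the QuarterClaspZoom Bolt character body that replaces emissive with animated glow tied, by parameter names, to an elimination count.
# A tech-art mask blends per-region Legs, Chest and Arms ranges that remap pre-skinned local height into gradients for panning bands.
# Bands run on three timestamp-driven transitions plus a looping time-based pan.
# Three-color ramps tint the panner and emissive-texture values, with a fade-in for the capped kill count.
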MaterialFunction /BRCosmetics/Characters/Player/Male/Medium/Bodies/M_MED_QuarterClaspZoom_Bolt/Materials/MF_QuarterClaspZoomBolt_TechArt
graph {nodes [
  n1 = GetMaterialAttributes(None=NamedRerouteUsage()) {attributeGetTypes [#1]}
  FunctionOutput(None=SetMaterialAttributes!(MaterialAttributes=@n1.MaterialAttributes, EmissiveColor=NamedRerouteUsage()) {attributeSetTypes [#1]})
  n2 = NamedRerouteUsage()
  n3 = NamedRerouteUsage()
  n4 = Subtract(A=Multiply(A=Multiply(A=@n2, B=@n3)) {constB=1.2}, B=NamedRerouteUsage())
  n5 = OneMinus(None=MaterialFunctionCall(Input=Abs(None=@n4), Input Low=Constant() {r=-0.2}, Input High=NamedRerouteUsage()) {materialFunction=&RemapValue_0-1}.Result)
  Multiply(A=@n5, B=ComponentMask(None=NamedRerouteUsage()) {r=true})
  n6 = NamedRerouteUsage()
  n7 = NamedRerouteUsage()
  n8 = Subtract(A=Multiply(A=Multiply(A=@n6, B=@n7)) {constB=1.2}, B=NamedRerouteUsage())
  n9 = OneMinus(None=MaterialFunctionCall(Input=Abs(None=@n8), Input Low=Constant() {r=-0.2}, Input High=NamedRerouteUsage()) {materialFunction=&RemapValue_0-1}.Result)
  Multiply(A=@n9, B=ComponentMask(None=NamedRerouteUsage()) {r=true})
  n10 = Frac(None=Multiply(A=Time(), B=ScalarParameter() {defaultValue=0.15, parameterName=ConstantPannerTimeMultiplier}))
  n11 = Subtract(A=Multiply(A=@n10) {constB=1.1}, B=NamedRerouteUsage())
  n12 = OneMinus(None=MaterialFunctionCall(
    Input=Abs(None=@n11)
    Input Low=Constant() {r=-0.3}
    Input High=ScalarParameter() {defaultValue=0.1333, parameterName=PanningGlowSizeConstant}
   ) {materialFunction=&RemapValue_0-1}.Result)
  Multiply(A=@n12, B=ComponentMask(None=NamedRerouteUsage()) {r=true})
  n13 = Multiply(A=NamedRerouteUsage()) {constB=0.4}
  Multiply(A=NamedRerouteUsage(), B=Multiply(A=@n13, B=@n13))
  n14 = NamedRerouteUsage()
  n15 = NamedRerouteUsage()
  n16 = Subtract(A=Multiply(A=Multiply(A=@n14, B=@n15)) {constB=1.2}, B=NamedRerouteUsage())
  n17 = OneMinus(None=MaterialFunctionCall(Input=Abs(None=@n16), Input Low=Constant() {r=-0.2}, Input High=NamedRerouteUsage()) {materialFunction=&RemapValue_0-1}.Result)
  Multiply(A=@n17, B=ComponentMask(None=NamedRerouteUsage()) {r=true})
  NamedRerouteDeclaration(None=FunctionInput() {inputName=In_MaterialAttributes, inputType=FunctionInput_MaterialAttributes, sortPriority=1}) {name=In_MaterialAttributes, nodeColor=<rgba 1 0.8438 0 1>}
  n18 = TextureSample() {
   samplerType=SAMPLERTYPE_LinearColor
   texture=/BRCosmetics/Characters/Player/Male/Medium/Bodies/M_MED_QuarterClaspZoom_Bolt/Textures/T_QuarterClaspZoomBolt_Body_FX2
  }
  NamedRerouteDeclaration(None=FunctionInput(Preview=@n18.RGB) {inputName=In_TechArtMask, sortPriority=1}) {name=In_TechArtMask, nodeColor=<rgba 0.5234 0 1 1>}
  NamedRerouteDeclaration(None=ScalarParameter() {defaultValue=1, group=QuarterClaspZoom, parameterName=TimeMultiplier, sortPriority=61}) {name=TimeMultiplier, nodeColor=<rgba 0.8281 0 1 1>}
  n19 = VectorParameter() {group=QuarterClaspZoom, parameterName=LowValueColor, sortPriority=80, defaultValue=<rgba 0.0885 0.0035 0.0124 1>}
  NamedRerouteDeclaration(None=@n19.RGB) {name=LowValueColor, nodeColor=<rgba 1 0 0.4922 1>}
  n20 = VectorParameter() {group=QuarterClaspZoom, parameterName=MidValueColor, sortPriority=81, defaultValue=<rgba 0.1826 0.0088 0.007 1>}
  NamedRerouteDeclaration(None=@n20.RGB) {name=MidValueColor, nodeColor=<rgba 0 0.8125 1 1>}
  n21 = VectorParameter() {group=QuarterClaspZoom, parameterName=HighValueColor, sortPriority=82, defaultValue=<rgba 0.9043 0.105 0.0629 1>}
  NamedRerouteDeclaration(None=@n21.RGB) {name=HighValueColor, nodeColor=<rgba 1 0.8672 0 1>}
  n22 = Add(A=NamedRerouteUsage(), B=NamedRerouteUsage())
  n23 = NamedRerouteUsage()
  n24 = NamedRerouteUsage()
  n25 = NamedRerouteUsage()
  n26 = NamedRerouteUsage()
  n27 = MaterialFunctionCall(A=@n24, B=@n25, C=@n26, Alpha=Saturate(None=Add(A=@n22, B=@n23))) {materialFunction=&Lerp_3Color}
  n28 = Multiply(A=NamedRerouteUsage()) {constB=4}
  n29 = NamedRerouteUsage()
  n30 = MaterialFunctionCall(A=@n24, B=@n25, C=@n26, Alpha=Saturate(None=Multiply(A=@n29) {constB=3.5})) {materialFunction=&Lerp_3Color}
  n31 = Multiply(A=NamedRerouteUsage()) {constB=0.5}
  NamedRerouteDeclaration(None=Add(A=Multiply(A=@n27.Result, B=@n28), B=Multiply(A=@n30.Result, B=@n31))) {name=GradientColorsFromPanner, nodeColor=<rgba 0.5469 0 1 1>}
  n32 = NamedRerouteUsage()
  n33 = Add(A=ComponentMask(None=@n32) {r=true}, B=ComponentMask(None=@n32) {g=true})
  n34 = ComponentMask(None=@n32) {b=true}
  NamedRerouteDeclaration(None=Multiply(A=Add(A=@n33, B=@n34), B=NamedRerouteUsage())) {name=EmissiveBaseValuesFromTexture, nodeColor=<rgba 0 1 0.2266 1>}
  n35 = NamedRerouteUsage()
  n36 = NamedRerouteUsage()
  n37 = NamedRerouteUsage()
  n38 = NamedRerouteUsage()
  n39 = Multiply(A=NamedRerouteUsage()) {constB=0.31}
  NamedRerouteDeclaration(None=Multiply(
    A=MaterialFunctionCall(A=@n35, B=@n36, C=@n37, Alpha=@n38) {materialFunction=&Lerp_3Color}.Result
    B=Multiply(A=@n39, B=@n39)
   )) {name=GradientColorsFromTexture, nodeColor=<rgba 0 1 0.6953 1>}
  n40 = Subtract(A=NamedRerouteUsage())
  NamedRerouteDeclaration(None=LinearInterpolate(A=Max(A=@n40) {constB=0}, B=NamedRerouteUsage(), Alpha=NamedRerouteUsage())) {name=NumberOfKillsFadesIn, nodeColor=<rgba 0.1953 0 1 1>}
  NamedRerouteDeclaration(None=ScalarParameter() {defaultValue=2, parameterName=EmissiveTextureMultiplier, sortPriority=71}) {name=EmissiveTextureMultiplier, nodeColor=<rgba 0.4531 1 0 1>}
  n41 = NamedRerouteUsage()
  n42 = NamedRerouteUsage()
  NamedRerouteDeclaration(None=VertexInterpolator(VS=MaterialFunctionCall(TimestampA=@n41, TransitionLength_In=@n42) {materialFunction=&MF_Timestamp_InOut_01}.Linear).PS) {name=PannerTimeFadeIn, nodeColor=<rgba 0.0274 0.2424 1 1>}
  n43 = LocalPosition() {localOrigin=InstancePreSkinning}
  n44 = ComponentMask(None=NamedRerouteUsage()) {g=true}
  n45 = LinearInterpolate(
   A=ScalarParameter() {parameterName="Input Low Legs(S)"}
   B=ScalarParameter() {defaultValue=100, parameterName="Input Low Chest(S)"}
   Alpha=Saturate(None=Multiply(A=@n44) {constB=12})
  )
  n46 = ScalarParameter() {defaultValue=80, parameterName="Input Low Arms(S)"}
  n47 = NamedRerouteUsage()
  n48 = Saturate(None=Multiply(A=ComponentMask(None=@n47) {b=true}) {constB=12})
  n49 = ComponentMask(None=NamedRerouteUsage()) {g=true}
  n50 = LinearInterpolate(
   A=ScalarParameter() {defaultValue=120, parameterName="Input High Legs(S)"}
   B=ScalarParameter() {defaultValue=300, parameterName="Input High Chest(S)"}
   Alpha=Saturate(None=Multiply(A=@n49) {constB=12})
  )
  n51 = ScalarParameter() {defaultValue=280, parameterName="Input High Arms(S)"}
  n52 = NamedRerouteUsage()
  n53 = Saturate(None=Multiply(A=ComponentMask(None=@n52) {b=true}) {constB=12})
  NamedRerouteDeclaration(None=MaterialFunctionCall(
    Input=VertexInterpolator(VS=@n43.Z).PS
    Input Low=LinearInterpolate(A=@n45, B=@n46, Alpha=@n48)
    Input High=LinearInterpolate(A=@n50, B=@n51, Alpha=@n53)
   ) {materialFunction=&RemapValue_0-1}.Result) {name=Gradients, nodeColor=<rgba 1 0 0.5156 1>}
  NamedRerouteDeclaration(None=ScalarParameter() {group=QuarterClaspZoom, parameterName=TimestampPanner, sortPriority=51}) {name=TimestampPanner, nodeColor=<rgba 1 0.3516 0 1>}
  NamedRerouteDeclaration(None=ScalarParameter() {defaultValue=1, group=QuarterClaspZoom, parameterName=TransitionLengthPanner, sortPriority=52}) {name=TransitionLengthPanner, nodeColor=<rgba 0.0312 0 1 1>}
  NamedRerouteDeclaration(None=@n12) {name=PanningGlowGradientConstant, nodeColor=<rgba 0.3594 1 0 1>}
  n54 = VectorParameter() {group=QuarterClaspZoom, parameterName=PannerLowValueColor, sortPriority=80, defaultValue=<rgba 0 0 0 1>}
  NamedRerouteDeclaration(None=@n54.RGB) {name=PannerLowValueColor, nodeColor=<rgba 1 0 0.4922 1>}
  n55 = NamedRerouteUsage()
  n56 = NamedRerouteUsage()
  NamedRerouteDeclaration(None=VertexInterpolator(VS=MaterialFunctionCall(TimestampA=@n55, TransitionLength_In=@n56) {materialFunction=&MF_Timestamp_InOut_01}.Linear).PS) {name=PannerTime, nodeColor=<rgba 0.0274 0.2424 1 1>}
  n57 = VectorParameter() {group=QuarterClaspZoom, parameterName=PannerMidValueColor, sortPriority=81, defaultValue=<rgba 0.1826 0.0088 0.007 1>}
  NamedRerouteDeclaration(None=@n57.RGB) {name=PannerMidValueColor, nodeColor=<rgba 0 0.8125 1 1>}
  n58 = VectorParameter() {group=QuarterClaspZoom, parameterName=PannerHighValueColor, sortPriority=82, defaultValue=<rgba 0.9043 0.105 0.0629 1>}
  NamedRerouteDeclaration(None=@n58.RGB) {name=PannerHighValueColor, nodeColor=<rgba 1 0.8672 0 1>}
  NamedRerouteDeclaration(None=Min(A=ScalarParameter() {defaultValue=1, parameterName=NumberOfKills}, B=ScalarParameter() {defaultValue=4, parameterName=MaxEliminations}) {constB=4}) {name=NumberOfKills, nodeColor=<rgba 0 1 0.8594 1>}
  NamedRerouteDeclaration(None=ScalarParameter() {group=QuarterClaspZoom, parameterName=TimestampFadeIn, sortPriority=51}) {name=TimestampFadeIn, nodeColor=<rgba 0.8112 1 0.0986 1>}
  n59 = NamedRerouteUsage()
  n60 = NamedRerouteUsage()
  NamedRerouteDeclaration(None=VertexInterpolator(VS=MaterialFunctionCall(TimestampA=@n59, TransitionLength_In=@n60) {materialFunction=&MF_Timestamp_InOut_01}.Linear).PS) {name=PannerTimeB, nodeColor=<rgba 0.0274 0.2424 1 1>}
  n61 = NamedRerouteUsage()
  n62 = NamedRerouteUsage()
  NamedRerouteDeclaration(None=VertexInterpolator(VS=MaterialFunctionCall(TimestampA=@n61, TransitionLength_In=@n62) {materialFunction=&MF_Timestamp_InOut_01}.Linear).PS) {name=PannerTimeC, nodeColor=<rgba 0.0274 0.2424 1 1>}
  NamedRerouteDeclaration(None=ScalarParameter() {group=QuarterClaspZoom, parameterName=TimestampPannerB, sortPriority=51}) {name=TimestampPannerB, nodeColor=<rgba 1 0.0236 0.7752 1>}
  NamedRerouteDeclaration(None=ScalarParameter() {defaultValue=1, group=QuarterClaspZoom, parameterName=TransitionLengthPannerB, sortPriority=52}) {name=TransitionLengthPannerB, nodeColor=<rgba 0.0623 1 0.8851 1>}
  NamedRerouteDeclaration(None=ScalarParameter() {group=QuarterClaspZoom, parameterName=TimestampPannerC, sortPriority=51}) {name=TimestampPannerC, nodeColor=<rgba 0.8112 1 0.0986 1>}
  NamedRerouteDeclaration(None=ScalarParameter() {defaultValue=1, group=QuarterClaspZoom, parameterName=TransitionLengthPannerC, sortPriority=52}) {name=TransitionLengthPannerC, nodeColor=<rgba 1 0.0137 0.0673 1>}
  NamedRerouteDeclaration(None=@n17) {name=PanningGlowGradientB, nodeColor=<rgba 0.3594 1 0 1>}
  NamedRerouteDeclaration(None=@n9) {name=PanningGlowGradientC, nodeColor=<rgba 0.3594 1 0 1>}
  NamedRerouteDeclaration(None=ScalarParameter() {defaultValue=1, group=QuarterClaspZoom, parameterName=TransitionLengthFadeIn, sortPriority=52}) {name=TransitionLengthFadeIn, nodeColor=<rgba 1 0.0137 0.0673 1>}
  NamedRerouteDeclaration(None=ScalarParameter() {defaultValue=0.1, group=QuarterClaspZoom, parameterName=PanningGlowSize, sortPriority=60}) {name=PanningGlowSize, nodeColor=<rgba 0.2891 1 0 1>}
  NamedRerouteDeclaration(None=@n5) {name=PanningGlowGradient, nodeColor=<rgba 0.3594 1 0 1>}
  NamedRerouteDeclaration(None=@n1.EmissiveColor) {name=In_EmissiveColor, nodeColor=<rgba 0.1953 0 1 1>}
  n63 = NamedRerouteUsage()
  n64 = NamedRerouteUsage()
  n65 = NamedRerouteUsage()
  n66 = Add(A=Add(A=ComponentMask(None=@n65) {r=true}, B=ComponentMask(None=@n65) {g=true}), B=ComponentMask(None=@n65) {b=true})
  NamedRerouteDeclaration(None=LinearInterpolate(A=Constant(), B=Add(A=@n63, B=@n64), Alpha=Saturate(None=@n66))) {name=Out_EmissiveColor, nodeColor=<rgba 0.125 1 0 1>}
  n67 = TextureObjectParameter() {
   group=QuarterClaspZoom
   parameterName=Emissive
   sortPriority=70
   texture=/BRCosmetics/Characters/Player/Male/Medium/Bodies/M_MED_QuarterClaspZoom_Bolt/Textures/T_QuarterClaspZoom_Bolt_Body_E
  }
  NamedRerouteDeclaration(None=@n67) {name=EmissiveTexture, nodeColor=<rgba 0 0.4844 1 1>}
  NamedRerouteUsage()
  NamedRerouteUsage()
  NamedRerouteUsage()
  NamedRerouteUsage()
  Subtract(A=@n10) {constB=0}
  TextureSample(UVs=TextureCoordinate(), Tex=NamedRerouteUsage()) {texture=&T_QuarterClaspZoom_Bolt_Body_E}
 ]}
