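# Post-processing-style AnimBlueprint for the Bolt variant of the QuarterClaspZoom male medium body cosmetic, layering secondary motion onto the attached parent mesh's pose.
# The copied pose runs through the body's Control Rig, fed a Sliding value that knee-slide begin and end events switch between 1 and 0.
# Two chained spring-driven Anim Dynamics nodes then simulate the left and right jacket bones, both using a spherical limit driven by thigh_r.
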AnimBlueprint /BRCosmetics/Characters/Player/Male/Medium/Bodies/M_MED_QuarterClaspZoom_Bolt/Meshes/M_MED_QuarterClaspZoom_Bolt_AnimBP
target_skeleton=&/M_MED_QuarterClaspZoom_Bolt_Skeleton
graphs [
 AnimGraph {nodes [
   n1 = Control Rig!(Source=Copy Pose From Mesh() {bUseAttachedParent=true}, Alpha, Sliding:double=Get Sliding() -> (Sliding:double)) -> (Pose) {controlRigAssetReference {blueprintRigClass=&/M_MED_QuarterClaspZoom_Bolt_CtrlRig.M_MED_QuarterClaspZoom_Bolt_CtrlRig_C}}
   n2 = Anim Dynamics - Bone: Base_jacket_l(ComponentPose=Local To Component(LocalPose=@n1), Alpha=1, ExternalForce="0,0,0") {
    angularDampingOverride=0.95
    angularSpringConstant=300
    bAngularSpring=true
    bGravityOverrideInSimSpace=true
    bLinearSpring=true
    bOverrideAngularDamping=true
    bUseGravityOverride=true
    gravityScale=0.35
    linearSpringConstant=50
    numSolverIterationsPostUpdate=2
    numSolverIterationsPreUpdate=8
    alphaBoolBlend {blendOption=Linear}
    boundBone {boneName=Base_jacket_l}
    componentAppliedLinearAccClamp=<5000 5000 5000>
    componentLinearAccScale=<0.5 0.5 0.5>
    componentLinearVelScale=<0.5 0.5 5>
    gravityOverride {z=-980}
    physicsBodyDefinitions [{
      collisionType=CustomSphere
      sphereCollisionRadius=4
      boundBone {boneName=Base_jacket_l}
      boxExtents=<12 12 12>
      constraintSetup {angularLimitsMax=<5 8 15>, angularLimitsMin=<-5 -8 -1>, angularTarget=<1 0 0>}
      localJointOffset=<0.7 0 0>
     }]
    sphericalLimits [{limitRadius=5, drivingBone {boneName=thigh_r}, sphereLocalOffset=<7.6962 -2.8316 1.2847>}]
   }
   n3 = Anim Dynamics - Bone: Base_jacket_r(ComponentPose=@n2, Alpha=1, ExternalForce="0,0,0") {
    angularDampingOverride=0.95
    angularSpringConstant=300
    bAngularSpring=true
    bGravityOverrideInSimSpace=true
    bLinearSpring=true
    bOverrideAngularDamping=true
    bUseGravityOverride=true
    gravityScale=0.35
    linearSpringConstant=50
    numSolverIterationsPostUpdate=2
    numSolverIterationsPreUpdate=8
    alphaBoolBlend {blendOption=Linear}
    boundBone {boneName=Base_jacket_r}
    componentAppliedLinearAccClamp=<5000 5000 5000>
    componentLinearAccScale=<0.5 0.5 0.5>
    componentLinearVelScale=<0.5 0.5 5>
    gravityOverride {z=-980}
    physicsBodyDefinitions [{
      collisionType=CustomSphere
      sphereCollisionRadius=4
      boundBone {boneName=Base_jacket_r}
      boxExtents=<12 12 12>
      constraintSetup {angularLimitsMax=<5 8 15>, angularLimitsMin=<-5 -8 -1>, angularTarget=<1 0 0>}
      localJointOffset=<0.7 0 0>
     }]
    sphericalLimits [{limitRadius=5, drivingBone {boneName=thigh_r}, sphereLocalOffset=<7.6962 -2.8316 1.2847>}]
   }
   Output Pose(Result=Component To Local(ComponentPose=@n3)) {layerGroup=None}
  ]}
 EventGraph {nodes [
   Event OnBeginKneeSliding() {is=title K2Node_Event}
   > Set Sliding(Sliding:double=1) -> (Output_Get:double)
   Event OnEndKneeSliding() {is=title K2Node_Event}
   > Set Sliding(Sliding=0)
   Event BlueprintUpdateAnimation()
  ]}
]
variables [Sliding {type=number}]
